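# Niagara effect for the BriskImp cosmetic companion's frost pickaxe, combining frost particles, glow lights and a weapon swing trail.
# Frost emits mist, snowflake and sparkle sprites plus a light from skeletal mesh sockets, with spawn groups switching per-particle settings and visibility tags.
# Light adds a socket-attached glow, while Weapon_Trail builds layered animation-trail ribbons.
# User floats control trail, mist and particle lifetime; time-of-day and front-end modifiers scale brightness.
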NiagaraSystem /CosmeticCompanions/FX/Builtins/BriskImp/NS_BriskImp_pickaxe_frost
dependencies {
 usedDataInterfaces [&NiagaraDataInterfaceCurve, &NiagaraDataInterfaceColorCurve, &NiagaraDataInterfaceSkeletalMesh]
 usedDynamicInputs [&SimulationPosition, &FNTimeOfDayColor]
 usedModules [
  &SystemState
  &self:SystemSpawnScript.SystemScriptSource.SystemScriptGraph.NiagaraNodeAssignment_1.SetVariables_#1
  &self:SystemSpawnScript.SystemScriptSource.SystemScriptGraph.NiagaraNodeAssignment_2.SetVariables_#2
 ]
}
emitters {
 Frost {
  emitterUpdateScript [
   &EmitterState(Life Cycle Mode=::NewEnumerator1, Loop Duration=10, UseLoopDelay=false)
   &SpawnBurst_Instantaneous(Spawn Group=3, Use Spawn Probability=false, UseLoopCountLimit=false)
   &SpawnRate(SpawnGroup=2, SpawnRate=15, Use Spawn Probability=false) {module=SpawnRate003}
   &SpawnRate(SpawnGroup=1, SpawnRate=15, Use Spawn Probability=false) {module=SpawnRate002}
   &SpawnRate(
    SpawnRate=&FloatFromCurve(
     FloatCurve=DataInterface() {propertyValues {Curve {keys [<at 0 1.3>,<at 0.3916 1.3>,<at 0.5968 3>,<at 0.9315 3>,<at 1 2.5>]}}}
     CurveIndex=$Emitter.NormalizedLoopAge
     Scale Curve=&Multiply_Float(A=30, B=$User.FrostyMistTrail_Rate)
    )
    Use Spawn Probability=false
   )
   &SpawnRate(SpawnRate=&Multiply_Float(A=15, B=$User.FrostyMistTrail_Rate), Use Spawn Probability=false) {enabled=false, module=SpawnRate001}
   &SpawnPerUnit(Max Movement Threshold=1, Module.Use Max Movement Threshold=true, Module.Use Movement Tolerance=true, Use Spawn Probability=false) {enabled=false}
  ]
  particleSpawnScript [
   /SetVariables(Particles.VisibilityTag=&SpawnGroup()) {isSetParameters=true}
   &InitializeParticle(
    Lifetime=&MergeHelper_FloatSwitch(
     0=$User.MistyTrail_Life
     1=&Multiply_Float(A=$User.Anim_Curve_particle_Life, B=&RandomRangeFloat(Minimum=0.3, Maximum=0.6))
     2=&Multiply_Float(A=$User.Anim_Curve_particle_Life, B=&RandomRangeFloat(Minimum=0.3, Maximum=0.6))
     3=2
     Index=&SpawnGroup()
     Num=::NewEnumerator2
    )
    Mass Mode=::NewEnumerator1
    Sprite Rotation Mode=::NewEnumerator2
    Sprite Size=&MergeHelper_Vector2Switch(
     0=&Vector2DFromFloat(Value=&RandomRangeFloat(Minimum=20, Maximum=35))
     1=&Vector2DFromFloat(Value=&RandomRangeFloat(Minimum=5, Maximum=10))
     2=&Vector2DFromFloat(Value=&RandomRangeFloat(Minimum=5, Maximum=10))
     3=?
     4=?
     Index=&SpawnGroup()
     Num=::NewEnumerator1
    )
    Sprite Size Mode=::NewEnumerator4
    UsePositionOffset=false
    Write Material Random=true
   ) {module=InitializeParticle001}
   &CameraOffset(Camera Offset Amount=10)
   &SkeletalMeshLocation(
    Bone Sampling Mode=::NewEnumerator4
    MaskSpawnGroup=false
    Position Sampling=::NewEnumerator1
    Sampled Position Offset=&MergeHelper_VectorSwitch(0=<5 0 0>, 1=<-2 3 5>, 2=?, 3=?, Index=$Particles.PartitionParticles001.Partition, Num=::NewEnumerator0)
    Skeletal Mesh=$System.Skeletal Mesh
    Source Movement Scale:Vector3f=<0 0 0>
   ) {enabled=false}
   &ApplyInitialForces()
   &PartitionParticles(Number Of Partitions=3)
   &SkeletalMeshLocation(
    Bone / Socket Index=$Particles.PartitionParticles.Partition
    MaskSpawnGroup=false
    Mesh Sampling Type=::NewEnumerator5
    Sampled Position Offset=&MergeHelper_VectorSwitch(0=<0 0 0>, 1=<0 0 0>, 2=<0 0 0>, 3=?, Index=$Particles.PartitionParticles.Partition, Num=::NewEnumerator1)
    Skeletal Mesh=$System.Skeletal Mesh
    Socket Sampling Mode=::NewEnumerator5
   ) {module=SkeletalMeshLocation001}
   &ShapeLocation(
    MaskSpawnGroup=false
    Override Seed=false
    Sphere Radius=&MergeHelper_FloatSwitch(0=10, 1=10, 2=8, 3=0, 4=0, Index=&SpawnGroup(), Num=::NewEnumerator3)
    UseOwnerScale=true
   )
  ]
  particleUpdateScript [
   &ParticleState()
   &GravityForce(Gravity=&MergeHelper_VectorSwitch(0=<0 0 -150>, 1=<0 0 -150>, 2=<0 0 -150>, 3=<0 0 0>, 4=<0 0 0>, Index=&SpawnGroup(), Num=::NewEnumerator3))
   &ScaleSpriteSize(Scale Factor=&Vector2DFromFloat(Value=&FloatFromCurve(
      FloatCurve=DataInterface() {propertyValues {Curve {defaultValue=0.5, keys [{interpMode=RCIM_Cubic, tangentMode=RCTM_User, time=0, value=0.75}, {tangentMode=RCTM_User, time=1, value=0.3}]}}}
      Scale Curve=1.5
     )))
   &Drag(
    Drag=&MergeHelper_FloatSwitch(
     0=2.5
     1=&RandomRangeFloat(Minimum=0.5, Maximum=2)
     2=&RandomRangeFloat(Minimum=0.5, Maximum=1)
     3=?
     Index=&SpawnGroup()
     Num=::NewEnumerator1
    )
    Use Linear Drag=true
    Use Rotational Drag=true
   )
   &InheritSourceMovement(ApplyForce=false, ApplyPosition=true, ApplyPositionFalloff=false, Source Velocity:Vector3f=$Emitter.InheritVelocityBlend)
   &InheritVelocity(Inherited Velocity Speed Limit=250, Source Speed Threshold=50, Speed Limit=true, UseSpeedThreshold=true)
   &PointAttractionForce(
    Attraction Radius=8
    Attraction Strength=1500
    Attractor Position:NiagaraPosition=$Particles.Initial.ShapeLocation.ShapePosition
    Kill Radius=5
    Kill Radius Overshoot Correction=5
    Kill Within Radius=true
    Module.Use Falloff=true
    Position To Attract:NiagaraPosition=$Particles.SkeletalMeshLocation.SampledPosition
    UseAttractorOffset=false
   )
   &SolveForcesAndVelocity(Clamp Velocity=false, Limit Acceleration=false)
   &DynamicMaterialParameters(
    Index 0 Param 1=0.5
    Module.Index0Float0Write=true
    Module.Index0Float1Write=true
    Module.Index0Float2Write=true
    Module.Index0Float3Write=true
   )
   /SetVariables(
    Particles.SnowFlake Color:LinearColor=<rgba 0.7669 0.9568 2 1>
    Particles.Snowflake Size:Vector2f=&Vector2DFromFloat(Value=&FloatFromCurve(
      FloatCurve=DataInterface() {propertyValues {Curve {
         defaultValue=0
         keys [
          {interpMode=RCIM_Cubic, tangentMode=RCTM_User, time=0, value=0}
          {arriveTangent=2, interpMode=RCIM_Cubic, leaveTangent=2, time=0.2204, value=0.8013}
          {interpMode=RCIM_Cubic, tangentMode=RCTM_User, time=0.5, value=1}
          {tangentMode=RCTM_User, time=1, value=0}
         ]
        }}}
      Scale Curve=&RandomRangeFloat(Minimum=3.5, Maximum=8)
     ))
   ) {isSetParameters=true}
   /SetVariables(
    Particles.Sparkle Color:LinearColor=&MakeLinearColorFromVectorAndFloat(
     "Vector (RGB)"=&Multiply_Vector(
      A=&RandomRangeVector(
       Minimum=&MakeVectorFromLinearColorRGB(Color=&ColorFromCurve(ColorCurve=DataInterface() {propertyValues {
           AlphaCurve {keys [<at 0.4411 1>]}
           BlueCurve {keys [<at 0 1>,<at 1 1 RCIM_Cubic>]}
           GreenCurve {keys [<at 0 0.8383>,<at 1 0.8402 RCIM_Cubic>]}
           RedCurve {keys [<at 0 0.6969>,<at 1 0.5991 RCIM_Cubic>]}
          }}))
       Maximum=&MakeVectorFromLinearColorRGB(Color=&ColorFromCurve(ColorCurve=DataInterface() {propertyValues {
           AlphaCurve {keys [<at 0.4411 1>]}
           BlueCurve {keys [<at 0 1 RCIM_Cubic>,<at 1 1 RCIM_Cubic>]}
           GreenCurve {keys [<at 0 0.7628 RCIM_Cubic>,<at 1 0.273 RCIM_Cubic>]}
           RedCurve {keys [<at 0 0.0681 RCIM_Cubic>,<at 1 0.1094 RCIM_Cubic>]}
          }}))
      )
      B=<3 3 3>
     )
     "Float (Alpha)"=&FloatFromCurve(
      FloatCurve=DataInterface() {propertyValues {Curve {
         defaultValue=1
         keys [
          {arriveTangent=9.0725, interpMode=RCIM_Cubic, leaveTangent=9.0725, tangentMode=RCTM_User, time=0, value=0}
          {arriveTangent=0.2227, interpMode=RCIM_Cubic, leaveTangent=0.2227, tangentMode=RCTM_User, time=0.1408, value=0.9847}
          {arriveTangent=-1.8564, interpMode=RCIM_Cubic, leaveTangent=-1.8564, tangentMode=RCTM_User, time=1, value=0}
         ]
        }}}
      Scale Curve=&FNTimeOfDayScalar(Night Scale=3.5, Morning Scale=5, Day Scale=5, Evening Scale=3.5)
     )
    )
    Particles.Sparkle DynParams:Vector4f=&MakeVector4(X=0.2, Y=1, Z=1, W=1)
    Particles.Sparkle Sprite Size:Vector2f=&Multiply_Vector2D(A=&RandomRangeVector2D(Minimum=<1.5 1.5>, Maximum=<3 3>), B=$Particles.SpriteSize)
   ) {isSetParameters=true}
   /SetVariables(
    Particles.Glow size:Vector2f=<5 5>
    Particles.Snowflake Glow color:LinearColor=&MakeLinearColorFromVectorAndFloat(
     "Vector (RGB)"=&Multiply_Vector(
      A=&RandomRangeVector(
       Minimum=&MakeVectorFromLinearColorRGB(Color=&ColorFromCurve(ColorCurve=DataInterface() {propertyValues {
           AlphaCurve {keys [<at 0 0 RCIM_Cubic>,<at 0.0335 0.5>,<at 1 0 RCIM_Cubic>]}
           BlueCurve {keys [<at 0 1 RCIM_Cubic>,<at 1 1 RCIM_Cubic>]}
           GreenCurve {keys [<at 0 0.4764 RCIM_Cubic>,<at 1 0.1837 RCIM_Cubic>]}
           RedCurve {keys [<at 0 0.0859 RCIM_Cubic>,<at 1 0 RCIM_Cubic>]}
          }}))
       Maximum=&MakeVectorFromLinearColorRGB(Color=&ColorFromCurve(ColorCurve=DataInterface() {propertyValues {
           AlphaCurve {keys [<at 0 0>,<at 0.1429 0.25>,<at 1 0>]}
           BlueCurve {keys [<at 0 0.7231 RCIM_Cubic>,<at 1 1 RCIM_Cubic>]}
           GreenCurve {keys [<at 0 0.7406 RCIM_Cubic>,<at 1 0.273 RCIM_Cubic>]}
           RedCurve {keys [<at 0 0.1921 RCIM_Cubic>,<at 1 0.1094 RCIM_Cubic>]}
          }}))
      )
      B=<1 1 1>
     )
     "Float (Alpha)"=&FloatFromCurve(
      FloatCurve=DataInterface() {propertyValues {Curve {
         defaultValue=1
         keys [
          {arriveTangent=9.0725, interpMode=RCIM_Cubic, leaveTangent=9.0725, tangentMode=RCTM_User, time=0, value=0}
          {arriveTangent=0.2227, interpMode=RCIM_Cubic, leaveTangent=0.2227, tangentMode=RCTM_User, time=0.1408, value=0.9847}
          {arriveTangent=-1.8564, interpMode=RCIM_Cubic, leaveTangent=-1.8564, tangentMode=RCTM_User, time=1, value=0}
         ]
        }}}
      Scale Curve=&FNTimeOfDayScalar(Night Scale=0.27, Morning Scale=0.25, Day Scale=0.25, Evening Scale=0.25)
     )
    )
   ) {isSetParameters=true}
   /SetVariables(Particles.Frosty mist Color:LinearColor=&MakeLinearColorFromVectorAndFloat(
     "Vector (RGB)"=&MakeVectorFromLinearColorRGB(Color=$System.TOD Snowy Mist)
     "Float (Alpha)"=&FloatFromCurve(
      FloatCurve=DataInterface() {propertyValues {Curve {
         defaultValue=1
         keys [
          {interpMode=RCIM_Cubic, tangentMode=RCTM_User, time=0, value=0}
          {interpMode=RCIM_Cubic, tangentMode=RCTM_User, time=0.0884, value=0.9718}
          {tangentMode=RCTM_User, time=0.7, value=0}
         ]
        }}}
      Scale Curve=&Multiply_Float(
       A=&FNTimeOfDayScalar(Night Scale=0.07, Morning Scale=0.03, Day Scale=0.1, Evening Scale=0.02)
       B=$User.FrostyMistAlpha
      )
     )
    )) {isSetParameters=true}
   &SubUVAnimation(
    End Frame=25
    Sub UV Play Rate:NiagaraFloat=&Multiply_Float(A=0.1, B=$Particles.NormalizedAge)
    SubUV Animation Mode:ENiagara_SubUVLookupMode=::NewEnumerator3
   ) {module=SubUVAnimation001}
   &SpriteRotationRate(Rotation Rate=&FloatFromCurve(
     FloatCurve=DataInterface() {propertyValues {Curve {keys [<at 0 0 RCIM_Cubic>, <at 0.3 1 RCIM_Cubic>, {arriveTangent=-1.7543, tangentMode=RCTM_User, time=1, value=0}]}}}
     Scale Curve=&RandomRangeFloat(Minimum=-360, Maximum=360)
    ))
   &FN_Light_Attributes(
    Color=&ColorFromCurve(ColorCurve=DataInterface() {propertyValues {AlphaCurve {keys []}, BlueCurve {keys [<at 0.5132 0.7305>]}, GreenCurve {keys [<at 0.5132 0.3813>]}, RedCurve {keys [<at 0.5132 0.1329>]}}})
    Create Color Attribute=true
    Create Exponent Attribute=true
    Create Position Attribute=true
    Create Radius Attribute=true
    Enable Particle Light Enable=true
    Enable Visibility Tag=true
    Exponent=3
    Position=<0 0 50>
    Radius=&Multiply_Float(
     A=60
     B=&FloatFromCurve(FloatCurve=DataInterface() {propertyValues {Curve {keys [<at 0 1.8>,<at 0.2 2.05>,<at 0.5 2.2>,<at 1 1.8>]}}})
    )
    Scale Color=&FloatFromCurve(
     FloatCurve=DataInterface() {propertyValues {Curve {keys [
         {arriveTangent=5.626, interpMode=RCIM_Cubic, leaveTangent=5.626, tangentMode=RCTM_User, time=0, value=0.6}
         {arriveTangent=0.0895, interpMode=RCIM_Cubic, leaveTangent=0.0895, tangentMode=RCTM_User, time=0.1, value=1}
         {arriveTangent=-0.8586, interpMode=RCIM_Cubic, leaveTangent=-0.8586, tangentMode=RCTM_User, time=0.815, value=0.72}
         {arriveTangent=-4.4162, interpMode=RCIM_Cubic, leaveTangent=-4.4162, tangentMode=RCTM_User, time=1, value=0}
        ]}}}
     Scale Curve=&Multiply_Float(
      A=&FN_FrontEndFloatModifier(Gameplay=20, Front End=5)
      B=&FNTimeOfDayScalar(Night Scale=0.5, Morning Scale=1, Evening Scale=0.8)
     )
    )
    VisibilityTag=3
   )
  ]
  rendererClasses [&NiagaraSpriteRendererProperties, &NiagaraLightRendererProperties]
  renderers [
   {rendererClass=&NiagaraSpriteRendererProperties}
   {rendererClass=&NiagaraSpriteRendererProperties, rendererIndex=1}
   {rendererClass=&NiagaraSpriteRendererProperties, rendererIndex=2}
   {rendererClass=&NiagaraSpriteRendererProperties, rendererIndex=3}
   {rendererClass=&NiagaraLightRendererProperties, rendererIndex=4}
  ]
 }
 Light {
  emitterUpdateScript [
   &EmitterState(Inactive Response=::NewEnumerator1, Life Cycle Mode=::NewEnumerator1, Loop Behavior=::NewEnumerator1, Loop Delay=1, UseLoopDelay=false)
   &SpawnBurst_Instantaneous(Spawn Count=4, Use Spawn Probability=false, UseLoopCountLimit=false)
   &SpawnRate(
    SpawnRate=&FloatFromCurve(
     FloatCurve=DataInterface() {propertyValues {Curve {keys [<at 0 1.3>,<at 0.3916 1.3>,<at 0.5968 3>,<at 0.9315 3>,<at 1 2.5>]}}}
     CurveIndex=$Emitter.NormalizedLoopAge
     Scale Curve=&Multiply_Float(A=50, B=$User.FrostyMistTrail_Rate)
    )
    Use Spawn Probability=false
   ) {enabled=false}
   &SpawnRate(SpawnRate=&Multiply_Float(A=15, B=$User.FrostyMistTrail_Rate), Use Spawn Probability=false) {enabled=false, module=SpawnRate001}
   &SpawnPerUnit(Max Movement Threshold=1, Module.Use Max Movement Threshold=true, Module.Use Movement Tolerance=true, Use Spawn Probability=false) {enabled=false}
  ]
  particleSpawnScript [
   /SetVariables(Particles.VisibilityTag=&SpawnGroup()) {isSetParameters=true}
   &InitializeParticle(
    Color Mode=::NewEnumerator0
    Lifetime=5
    Mass Mode=::NewEnumerator1
    Sprite Rotation Mode=::NewEnumerator2
    UsePositionOffset=false
    Write Material Random=true
   ) {module=InitializeParticle001}
   &ApplyInitialForces()
  ]
  particleUpdateScript [
   &ParticleState()
   &SkeletalMeshLocation(
    MaskSpawnGroup=false
    Mesh Sampling Type=::NewEnumerator5
    Orientation Sampling=::NewEnumerator1
    Skeletal Mesh=$System.Skeletal Mesh
    Socket Sampling Mode=::NewEnumerator5
   )
   &SolveForcesAndVelocity(Clamp Velocity=false, Limit Acceleration=false)
   &FN_Light_Attributes(
    Color=<rgba 0.0938 0.2692 0.5156 1>
    Create Color Attribute=true
    Create Exponent Attribute=true
    Create Radius Attribute=true
    Enable Particle Light Enable=true
    Enable Visibility Tag=true
    Exponent=3
    Light Enabled=true
    Radius=&Multiply_Float(
     A=30
     B=&FloatFromCurve(FloatCurve=DataInterface() {propertyValues {Curve {keys [<at 0 1.8>,<at 0.2 2.05>,<at 0.5 2.2>,<at 1 1.8>]}}})
    )
    Scale Color=&FloatFromCurve(
     FloatCurve=DataInterface() {propertyValues {Curve {keys [
         {arriveTangent=5.626, interpMode=RCIM_Cubic, leaveTangent=5.626, tangentMode=RCTM_User, time=0, value=0.6}
         {arriveTangent=0.0895, interpMode=RCIM_Cubic, leaveTangent=0.0895, tangentMode=RCTM_User, time=0.1, value=1}
         {arriveTangent=-0.8586, interpMode=RCIM_Cubic, leaveTangent=-0.8586, tangentMode=RCTM_User, time=0.815, value=0.72}
         {arriveTangent=-4.4162, interpMode=RCIM_Cubic, leaveTangent=-4.4162, tangentMode=RCTM_User, time=1, value=0}
        ]}}}
     Scale Curve=&Multiply_Float(
      A=&FN_FrontEndFloatModifier(Gameplay=5, Front End=5)
      B=&FNTimeOfDayScalar(Night Scale=0.65, Morning Scale=1, Day Scale=0.75, Evening Scale=0.8)
     )
    )
   )
  ]
  rendererClasses [&NiagaraLightRendererProperties]
  renderers [{rendererClass=&NiagaraLightRendererProperties, rendererIndex=0}]
 }
 Weapon_Trail {
  emitterUpdateScript [
   &EmitterState(Inactive Response=::NewEnumerator1, Life Cycle Mode=::NewEnumerator1, Loop Behavior=::NewEnumerator1, Loop Duration=4.48, UseLoopDelay=false)
   &FN_SpawnAnimTrail(Spawn Interval=0)
  ]
  particleSpawnScript [
   &PartitionParticles(Write Partition to Ribbon ID=::NewEnumerator0)
   &ApplyInitialForces()
   &FN_InitializeAnimTrailBoneIndecies()
   &SkeletalMeshLocation(
    Bone / Socket Index=$Particles.PartitionParticles.Partition
    Bone Sampling Mode=::NewEnumerator4
    MaskSpawnGroup=false
    Skeletal Mesh=$System.Skeletal Mesh
   ) {enabled=false}
   &AddVelocity(Velocity=<1500 0 0>) {enabled=false}
   &FN_InitializeAnimTrail(
    AnimTrailWidth=8
    Lifetime=&Multiply_Float(
     A=&FloatFromCurve(
      FloatCurve=DataInterface() {propertyValues {Curve {
         defaultValue=1
         keys [{interpMode=RCIM_Cubic, tangentMode=RCTM_User, time=0, value=1}, {interpMode=RCIM_Cubic, tangentMode=RCTM_User, time=0.4905, value=-0.0601}]
        }}}
      Scale Curve=0.9
     )
     B=$User.TrailFX_On
    )
    SkeletalMesh=$System.Skeletal Mesh
   )
  ]
  particleUpdateScript [
   &ParticleState()
   &FN_UpdateAnimTrail(SkeletalMesh=$System.Skeletal Mesh)
   &ScaleColor(
    Scale Alpha=&FloatFromCurve(
     FloatCurve=DataInterface() {propertyValues {Curve {keys [{arriveTangent=-3.0806, interpMode=RCIM_Cubic, leaveTangent=-3.0806, tangentMode=RCTM_User, time=0, value=1}, <at 0.85 0.2071 RCIM_Cubic>]}}}
     Scale Curve=0.5
    )
    Scale RGB=&MakeVectorFromLinearColorRGB(Color=&ColorFromCurve(ColorCurve=DataInterface() {propertyValues {
        AlphaCurve {keys [<at 0 0.8 RCIM_Cubic>,<at 0.4686 0 RCIM_Cubic>]}
        BlueCurve {keys [<at 0 0.2 RCIM_Cubic>]}
        GreenCurve {keys [<at 0 0.1694 RCIM_Cubic>]}
        RedCurve {keys [<at 0 0.1319 RCIM_Cubic>]}
       }}))
    ScaleA=true
    ScaleRGB=true
   ) {module=ScaleColor003}
   &ScaleRibbonWidth(Ribbon Width Scale=&FloatFromCurve(FloatCurve=DataInterface() {propertyValues {Curve {
        defaultValue=1
        keys [
         {arriveTangent=0.0733, interpMode=RCIM_Cubic, leaveTangent=0.0733, tangentMode=RCTM_User, time=0, value=1}
         {arriveTangent=-0.0405, interpMode=RCIM_Cubic, leaveTangent=-0.0405, tangentMode=RCTM_User, time=0.7, value=0.3}
        ]
       }}}))
   &InheritSourceMovement(ApplyForce=false, ApplyPosition=true, ApplyPositionFalloff=false)
   &ScaleColor(
    Scale Alpha=0.7
    Scale RGB=&VectorFromFloat(Value=&FNTimeOfDayScalar(Night Scale=5, Morning Scale=5, Day Scale=5, Evening Scale=5))
    ScaleA=false
    ScaleRGB=true
   ) {module=ScaleColor001}
   &Drag(Drag=5, Use Linear Drag=true, Use Rotational Drag=false)
   &SolveForcesAndVelocity(Clamp Velocity=false, Limit Acceleration=false)
   /SetVariables(
    Particles.EnergyColor:LinearColor=/DealershipSharedLibrary/FX/Niagara/DynamicInputs/MultiplyLinearColorByVectorAndFloat(
     Color to Scale:LinearColor=&ColorFromCurve(ColorCurve=DataInterface() {propertyValues {
        AlphaCurve {keys [<at 0 0.5 RCIM_Cubic>,<at 0.9052 0 RCIM_Cubic>]}
        BlueCurve {keys [<at 0 1 RCIM_Cubic>,<at 0.2155 1 RCIM_Cubic>]}
        GreenCurve {keys [<at 0 0.7939 RCIM_Cubic>,<at 0.2155 0.6028 RCIM_Cubic>]}
        RedCurve {keys [<at 0 0.6135 RCIM_Cubic>,<at 0.2155 0.2552 RCIM_Cubic>]}
       }})
     RGB Scale:Vector3f=<1 1 1>
     Alpha Scale:NiagaraFloat=&FloatFromCurve(
      FloatCurve=DataInterface() {propertyValues {Curve {keys [{arriveTangent=-3.0806, interpMode=RCIM_Cubic, leaveTangent=-3.0806, tangentMode=RCTM_User, time=0, value=1}, <at 0.85 0.2071 RCIM_Cubic>]}}}
      Scale Curve=&Multiply_Float(
       A=&FNTimeOfDayScalar(Night Scale=0.7, Morning Scale=0.75, Day Scale=0.57, Evening Scale=0.75)
       B=$User.Prop_slash_Brightness
      )
     )
    )
    Particles.scratchy trail color:LinearColor=&MultiplyLinearColorByVectorAndFloat(
     Color to Scale:LinearColor=&ColorFromCurve(ColorCurve=DataInterface() {propertyValues {
        AlphaCurve {keys [<at 0 0.5 RCIM_Cubic>,<at 0.9052 0 RCIM_Cubic>]}
        BlueCurve {keys [<at 0 1 RCIM_Cubic>,<at 0.9943 1 RCIM_Cubic>]}
        GreenCurve {keys [<at 0 0.7939 RCIM_Cubic>,<at 0.9943 0.6028 RCIM_Cubic>]}
        RedCurve {keys [<at 0 0.6135 RCIM_Cubic>,<at 0.9943 0.2552 RCIM_Cubic>]}
       }})
     RGB Scale:Vector3f=<1 1 1>
     Alpha Scale:NiagaraFloat=&FloatFromCurve(
      FloatCurve=DataInterface() {propertyValues {Curve {keys [{arriveTangent=-3.0806, interpMode=RCIM_Cubic, leaveTangent=-3.0806, tangentMode=RCTM_User, time=0, value=1}, <at 0.85 0.2071 RCIM_Cubic>]}}}
      Scale Curve=&Multiply_Float(
       A=&FNTimeOfDayScalar(Night Scale=0.75, Morning Scale=0.75, Day Scale=0.57, Evening Scale=0.75)
       B=$User.Prop_slash_Brightness
      )
     )
    )
   ) {isSetParameters=true}
   /SetVariables(
    Particles.Snow dots Width:NiagaraFloat=&FloatFromCurve(FloatCurve=DataInterface() {propertyValues {Curve {keys [<at 0 1>,<at 1 0>]}}}, Scale Curve=12)
    Particles.scratchy trail width:NiagaraFloat=&FloatFromCurve(
     FloatCurve=DataInterface() {propertyValues {Curve {keys [<at 0 1>,<at 1 0>]}}}
     Scale Curve=&Multiply_Float(A=60, B=$User.Trail_Width)
    )
    Particles.snow filler width:NiagaraFloat=&FloatFromCurve(
     FloatCurve=DataInterface() {propertyValues {Curve {keys [<at 0 1>,<at 1 0>]}}}
     Scale Curve=&Multiply_Float(A=25, B=$User.Trail_Width)
    )
   ) {isSetParameters=true}
   /SetVariables(Particles.SnowColor:LinearColor=&MakeLinearColorFromVectorAndFloat(
     "Vector (RGB)"=&MakeVectorFromLinearColorRGB(Color=&ColorFromCurve(ColorCurve=DataInterface() {propertyValues {
         AlphaCurve {keys [<at 0 0.5 RCIM_Cubic>,<at 0.9052 0 RCIM_Cubic>]}
         BlueCurve {keys [<at 0 5 RCIM_Cubic>,<at 0.1985 1 RCIM_Cubic>]}
         GreenCurve {keys [<at 0 5 RCIM_Cubic>,<at 0.1985 0.8023 RCIM_Cubic>]}
         RedCurve {keys [<at 0 5 RCIM_Cubic>,<at 0.1985 0.5438 RCIM_Cubic>]}
        }}))
     "Float (Alpha)"=&FloatFromCurve(
      FloatCurve=DataInterface() {propertyValues {Curve {keys [<at 0 1>,<at 1 0>]}}}
      Scale Curve=&FNTimeOfDayScalar(Night Scale=1, Morning Scale=1, Evening Scale=1)
     )
    )) {isSetParameters=true}
  ]
  rendererClasses [&NiagaraRibbonRendererProperties]
  renderers [
   {rendererClass=&NiagaraRibbonRendererProperties}
   {rendererClass=&NiagaraRibbonRendererProperties, rendererIndex=1}
   {rendererClass=&NiagaraRibbonRendererProperties, rendererIndex=2}
   {rendererClass=&NiagaraRibbonRendererProperties, rendererIndex=3}
   {rendererClass=&NiagaraRibbonRendererProperties, rendererIndex=4}
  ]
 }
}
summary {
 systemName=NS_BriskImp_pickaxe_frost
 emitters [
  {emitterName=Frost, rendererClasses [&NiagaraSpriteRendererProperties, &NiagaraLightRendererProperties]}
  {emitterName=Weapon_Trail, rendererClasses [&NiagaraRibbonRendererProperties]}
  {emitterName=Light, rendererClasses [&NiagaraLightRendererProperties]}
 ]
 userVariables [
  User.Anim_Curve_particle_Life {type=NiagaraFloat, defaultValue=0, description=""}
  User.bIsFrontEnd {type=NiagaraBool}
  User.bIsFrontEndPreview {type=NiagaraBool}
  User.FrostyMistAlpha {type=NiagaraFloat, defaultValue=1, description=""}
  User.FrostyMistTrail_Rate {type=NiagaraFloat, defaultValue=1, description=""}
  User.MistyTrail_Life {type=NiagaraFloat, defaultValue=1, description=""}
  User.Prop_slash_Brightness {type=NiagaraFloat, defaultValue=0, description=""}
  User.RibbonIDOffset {type=NiagaraInt32}
  User.Trail_Width {type=NiagaraFloat, defaultValue=1, description=""}
  User.TrailFX_On {type=NiagaraFloat, defaultValue=0, description=""}
 ]
}
